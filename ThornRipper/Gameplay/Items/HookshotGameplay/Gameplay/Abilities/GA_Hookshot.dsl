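; Gameplay ability for the hookshot item: winds up, throws a roped projectile, and applies pull and damage effects to what it hits.
; Montage notifies drive readiness, throwing and ending, with early primary-press queuing.
; Hotfix flags select predicted, validated client-transform or server spawning; server homing prefers direct aim, then the client-suggested target, with soft-homing fallback.
; The server resolves hits or range expiry, then RPCs the client to play success or failure feedback.
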(event Ability|EventActivateAbility
  ;[0.2719 1 0] ActivateAbility
  (Utilities|FlowControl|Sequence
    (:then_0
      (Variables|Transient|SetbAbilityEnded false)
      ;;[0.09063 1 0.04957] Cache Hotfix values
      (Variables|Transient|SetCachedHookshotDistance (ScalableFloat|GetValueatLevel (Variables|Hotfix|GetHotfix_HookshotDistance))))
    (:then_1
      ;;[0 0.9745 1] ActivateAbility synchronous logic.
      (bind (_asynctaskproxy _notifytag) (Ability|Tasks|PlayMontageAndWaitNotify "Task_PlayMontage" (Variables|General|GetHookshotMontage) 1.0 "None" false)
        (:then
          (Variables|Transient|AsyncTask|SetAsyncTask_PlayMontage _asynctaskproxy)
          (Ability|HasAuthority
            (:True
              _)
            (:False
              (bind _asynctaskproxy_2 (Ability|Tasks|WaitInputRelease false)
                (:then
                  (Variables|Transient|AsyncTask|SetAsyncTask_WaitInputReleased _asynctaskproxy_2))
                (:OnRelease
                  (if (not (Variables|Transient|GetbHookshotThrown))
                    ;;[1 0.4635 0.4635] End Ability
                    (Ability|EndAbility :!label _label)))) ; Only on the client.
            ))
          ;;[0.1614 0.04942 0.1667] Set gameplay event to see if player queues the throw.
          (bind _asynctaskproxy_3 (Ability|Tasks|WaitGameplayEvent "(TagName=\"Event.Primary.Pressed\")" 0 true)
            (:then
              (Variables|Transient|AsyncTask|SetAsyncTask_EarlyPrimaryQueued _asynctaskproxy_3)
              ;;[1 0.9528 0.1882] Zoom the camera in with the targeting camera mode.
              (Camera|PushCameraMode (Variables|Camera|GetTargetingCameraModeClass)))
            (:EventReceived
              (Variables|Transient|SetbEarlyPrimaryQueued true))))
        (:OnCompleted
          _)
        (:OnBlendOut
          _)
        (:OnInterrupted
          _)
        (:OnCancelled
          _)
        (:OnNotifyBegin
          (CallFunction|OnMontageNotifyBegin _notifytag)))
      (goto _label))))

(event Custom|On%20Montage%20Notify%20Begin
  (params (InputPin :type Gameplay%20Tag%20Structure))
  ;[0.07265 0.2448 0.1537 0.5] Montage Notifiy Begin
  (if (GameplayTags|IsGameplayTagValid InputPin)
    ;;[0.5885 0.06801 0.5711] On Hookshot Windup complete
    (if (== InputPin (Variables|General|GetHookshotReadyAnimNotifyTag))
      ;;;[0.1719 0.07451 0.05561] Applying gameplay effect to the player that indicates the player is ready to throw the hook.
      (bind _returnvalue (Ability|ApplyGameplayEffectToOwner (Variables|General|GetPlayerHookshotReadyGameplayEffectClass)))
      (Variables|Transient|SetPlayerHookshotReadyGameplayEffectHandle _returnvalue)
      (bind _asynctask_earlyprimaryqueued (Variables|Transient|AsyncTask|GetAsyncTask_EarlyPrimaryQueued)
        (:then
          (GameplayTasks|EndTask _asynctask_earlyprimaryqueued))
        (:else
          _))
      (if (not (Variables|Transient|GetbEarlyPrimaryQueued))
        (bind _asynctaskproxy (Ability|Tasks|WaitGameplayEvent "(TagName=\"Event.Primary.Pressed\")" 0 true)
          (:then
            (Variables|Transient|AsyncTask|SetAsyncTask_PrimaryPressed _asynctaskproxy))
          (:EventReceived
            _)))
      (CallFunction|OnPrimaryTriggered)
      (elif (== InputPin (Variables|General|GetThrowAnimNotifyTag))
        ;;[0.07254 0.2396 0.1245 0.5] Spawn Projectile
        (CallFunction|SpawnHookshotProjectile)
        (Variables|Transient|SetbHookshotFired true)
        (Camera|PlayThrowCameraShake)
        (elif (== InputPin (Variables|General|GetMontageEndAnimNotifyTag))
          ;;[0.2135 0.09303 0.01434] End the ability when there's a notify on the montage that says to do so.
          (Ability|EndAbility))))))

(event Custom|On%20Primary%20Triggered
  (bind _throwcameramodeclass (Variables|Camera|GetThrowCameraModeClass))
  (bind _throwmontagesectionname (Variables|General|GetThrowMontageSectionName))
  ;[0.165 0.2109 1] On primary attack pressed, jump to the Throw section of the animation - this eventually ends up spawning the hookshot projectile.
  (Utilities|FlowControl|Sequence
    (:then_0
      (if (!= _throwmontagesectionname "")
        (Ability|Animation|MontageJumptoSection _throwmontagesectionname))
      (Camera|PopCameraMode) ; Removes the targeting camera mode.
      (if (Utilities|IsValidClass _throwcameramodeclass)
        (Camera|PushCameraMode _throwcameramodeclass) ; Push the throw camera mode class.
      )
      (GameplayTasks|EndTask (Variables|Transient|AsyncTask|GetAsyncTask_PrimaryPressed))
      (Ability|RemoveGameplayEffectFromOwnerWithHandle (Variables|Transient|GetPlayerHookshotReadyGameplayEffectHandle))
      (Variables|Transient|SetbHookshotThrown true))
    (:then_1
      ;;[0.04043 0.2139 0.3281] On the client, run a targeting trace as soon as the player makes the decision to throw the hook. Trace helps figure out the actor that the player was aiming for.
      ;; RPC the target actor found through the trace to the server. The server will use this target actor as the suggested actor to home the hookshot projectile towards.
      (Ability|HasAuthority
        (:True
          (if (General|IsBot)
            _))
        (:False
          (if (ScalableFloat|GetValueasBool (Variables|Hotfix|GetHotfix_UseClientSuggestedTarget))
            _)))
      (bind (_asynctaskproxy _targetdata) (Ability|Tasks|PerformTargetingAbilityTask "Task_HookshotTargeting_Trace_Client" (Variables|Targeting|GetTargetingPresetHookshotTargeting) false true)
        (:then
          (Variables|Transient|AsyncTask|SetAsyncTask_HookshotTargetTraceClient _asynctaskproxy))
        (:Targeted
          _)
        (:Cancelled
          _))
      (bind _targetactor (Homing|GetHomingTargetfromTargetData _targetdata))
      (bind _bvalid (Homing|IsValidHomingTarget _targetactor))
      (if _bvalid
        (Development|PrintString (Utilities|String|Append "Hookshot: TargetTrace Client : " (Utilities|GetDisplayName _targetactor)) true false "(R=1.000000,G=0.482353,B=0.000000,A=1.000000)" 1e+01 "Hookshot_EarlyTrace")
        (Utilities|IsValid self
          (:Is%20Valid
            (CallFunction|ServerSetClientSuggestedHomingTarget _targetactor) ; Let the server know about the target actor that was \nunder the player's reticle when they threw the hook.
          ))
        (else
          (Ability|HasAuthority
            (:False
              (if (ScalableFloat|GetValueasBool (Variables|Hotfix|GetHotfix_SuggestClientTargetingComponentHomingTargetActor))
                (bind _targetactor_2 (Homing|GetClientSuggestedHomingTargetfromTargetingComponent))
                (Utilities|IsValid _targetactor_2
                  (:Is%20Valid
                    (Utilities|IsValid self
                      (:Is%20Valid
                        (CallFunction|ServerSetClientSuggestedHomingTarget _targetactor_2) ; Let the server know about the target actor that was \nunder the player's reticle when they threw the hook.
                      ))))))))))))

(event Custom|Spawn%20Hookshot%20Projectile
  ;[0.1421 0.5625 0.5337 0.5] Projectile
  ;;[0.0541 0.2865 0.07439] Spawn the projectile.
  ;;; Calculate the spawn location and direction of the projectile.
  (bind (_spawnlocation _spawndirection) (Projectile|CalculateProjectileSpawnLocationandRotation))
  (Variables|Transient|SetProjectileSpawnLocation _spawnlocation)
  (Variables|Transient|SetProjectileSpawnDirection _spawndirection)
  (if (ScalableFloat|GetValueasBool (Variables|Hotfix|GetHotfix_ClientPredictiveProjectile))
    (bind _returnvalue (Projectile|SpawnClientPredictedProjectile (Variables|General|GetHookshot%20Predictive%20Projectile%20Class) (CallFunction|GetAvatarasFortPlayerPawn) (Utilities|Struct|MakeSpawnProjectileParams (Variables|Transient|GetProjectileSpawnLocation) (Variables|Transient|GetProjectileSpawnDirection) (Variables|Transient|GetProjectileSpawnDirection) "(TargetSelection=(List=,bStopAtFirstSuccess=True,bKeepCheckingListOnIndestructibleHit=False,AbilityTargetBlockedGameplayTags=(GameplayTags=),bUseWeaponRanges=True,bUseMaxYawAngleToTarget=False,MaxYawAngleToTarget=0.000000),TargetGameplayEffectSpecs=,OwnerGameplayEffectSpecs=,ActivationCues=(GameplayTags=),ImpactCues=(GameplayTags=),ImpactNoiseRange=0.000000,FlyByNoiseRange=0.000000,bOverrideChargeMagnitude=False,ChargeMagnitudeOverrideValue=0.000000)" "(TargetSelection=(List=,bStopAtFirstSuccess=True,bKeepCheckingListOnIndestructibleHit=False,AbilityTargetBlockedGameplayTags=(GameplayTags=),bUseWeaponRanges=True,bUseMaxYawAngleToTarget=False,MaxYawAngleToTarget=0.000000),TargetGameplayEffectSpecs=,OwnerGameplayEffectSpecs=,ActivationCues=(GameplayTags=),ImpactCues=(GameplayTags=),ImpactNoiseRange=0.000000,FlyByNoiseRange=0.000000,bOverrideChargeMagnitude=False,ChargeMagnitudeOverrideValue=0.000000)" false 0 (ScalableFloat|GetValueatLevel (Variables|Hotfix|GetHotfix_ProjectileSpeed)) 0.0 0.0)))
    (bind _asb_prj_hookshot (Utilities|Casting|CastToB_Prj_Hookshot _returnvalue)
      (:then
        (CallFunction|OnProjectileCreated _asb_prj_hookshot)
        (Default|BindEventtoOnProjectileStopped _asb_prj_hookshot (EventDispatchers|CreateEvent :!event Custom|On%20Projectile%20Stopped) :!class /ThornRipper/Gameplay/Items/HookshotGameplay/Gameplay/B_Prj_Hookshot.B_Prj_Hookshot_C)
        (Ability|HasAuthority
          (:True
            (Game|BindEventtoOnDestroyed _asb_prj_hookshot (EventDispatchers|CreateEvent :!event Custom|On%20Projectile%20Destroyed)))))
      (:CastFailed
        _))
    (elif (ScalableFloat|GetValueasBool (Variables|Hotfix|GetHotfix_ClientRPCProjectileSpawnTransform))
      ;;; Send the location and direction info directly from the client to the server to use.
      (Ability|HasAuthority
        (:True
          (if (General|IsBot)
            _))
        (:False
          _))
      (CallFunction|ServerSpawnProjectileatTransform (Variables|Transient|GetProjectileSpawnLocation) (Variables|Transient|GetProjectileSpawnDirection))
      (else
        (Ability|HasAuthority
          (:True
            (bind (_asynctaskproxy _projectiledata) (Projectile|Tasks|SpawnProjectileAndWait :Class (Variables|General|GetHookshotProjectileClass) :RequestedBy (CallFunction|GetAvatarasFortPlayerPawn) :SpawnLocation (Variables|Transient|GetProjectileSpawnLocation) :SpawnRotation (Variables|Transient|GetProjectileSpawnDirection) :SpawnDirection (Variables|Transient|GetProjectileSpawnDirection) :InitialSpeed (ScalableFloat|GetValueatLevel (Variables|Hotfix|GetHotfix_ProjectileSpeed)) :GravityScale 0.0 :bAllowProjectileSpawnWhenDead false :bAllowProjectileSpawnWhenDBNO false)
              (:then
                (Variables|Transient|AsyncTask|SetAsyncTask_SpawnProjectile _asynctaskproxy))
              (:Created
                (bind (_projectiledata_spawnedprojectile _projectiledata_hits) (Utilities|Struct|BreakProjectileEventData _projectiledata))
                (bind _asb_prj_hookshot_2 (Utilities|Casting|CastToB_Prj_Hookshot _projectiledata_spawnedprojectile)
                  (:then
                    (CallFunction|OnProjectileCreated _asb_prj_hookshot_2))
                  (:CastFailed
                    _)))
              (:Stopped
                (CallFunction|OnProjectileStopped _projectiledata_hits))
              (:Destroyed
                (CallFunction|OnProjectileDestroyed 0))))))))
  ;;;[1 0 0] Warning: Failed to create projectile.
  (Fortnite|FortShippingLog self "Failed to create hookshot projectile." true)
  (Ability|EndAbility))

(event Custom|Setup%20Homing%20Data
  ;[0.1615 0.00925 0.1377] Set up homing on the projectile
  (if (ScalableFloat|GetValueasBool (Variables|Hotfix|GetHotFix_DelayHomingSettingByOneFrame))
    (Utilities|FlowControl|DelayUntilNextTick))
  (Ability|HasAuthority
    (:True
      ;;[0.01349 0.0625 0.02134] First check to see if there's any target under the reticle. If so, home to that target.
      (bind (_asynctaskproxy _targetdata) (Ability|Tasks|PerformTargetingAbilityTask "Task_HookshotTargeting_Trace" (Variables|Targeting|GetTargetingPresetHookshotTargeting) true)
        (:then
          (Variables|Transient|AsyncTask|SetAsyncTask_HookshotTargetTrace _asynctaskproxy))
        (:Targeted
          _)
        (:Cancelled
          _))
      (Variables|Transient|GetHookshotProjectile
        (:then
          (bind _targetactor (Homing|GetHomingTargetfromTargetData _targetdata))
          (bind _bvalid (Homing|IsValidHomingTarget _targetactor))
          (if _bvalid
            (Development|PrintString "Hookshot: Direct aim found." true false "(R=1.000000,G=0.482353,B=0.000000,A=1.000000)" 1e+01 "Hookshot")
            ;;;[0.02158 0.1036 0.1406] Set up strong homing settings on the projectile since the player has directly aimed at the target.
            (Projectile|SetHomingData (Variables|Transient|GetHookshotProjectile) (Variables|Transient|GetHookshotProjectile) (Utilities|Struct|MakeProjectileHomingData "LockOn" (ScalableFloat|GetValueatLevel (Variables|Hotfix|Homing|Direct|GetHotfix_HomingDirectAim_MinTurnSpeed)) (ScalableFloat|GetValueatLevel (Variables|Hotfix|Homing|Direct|GetHotfix_HomingDirectAim_MaxTurnSpeed)) (ScalableFloat|GetValueatLevel (Variables|Hotfix|Homing|Direct|GetHotfix_HomingDirectAim_TurnSpeedRampTime)) (ScalableFloat|GetValueatLevel (Variables|Hotfix|Homing|Direct|GetHotfix_HomingDirectAim_DistanceThreshold)) -1 _targetactor 0 "0.000000,0.000000,0.000000" (Homing|ClampHomingRelativeOffset (Math|Transform|InverseTransformLocation (Transformation|GetActorTransform _targetactor) (.ImpactPoint (Collision|BreakHitResult (Ability|TargetData|GetHitResultfromTargetData _targetdata))))) "Relative" true) :!class /Script/FortniteGame.FortProjectileMovementInterface)
            ;;;;[0.224 0.04469 0.2092] Update the distance the projectile travels since we have an accurate aim.
            ;;;; This will help hit targets that are backpedalling away
            (Variables|/Script/FortniteGame.FortProjectileBase|SetMaxSimulationDistanceSquaredOverride (ScalableFloat|GetValueatLevel (Variables|Hotfix|GetHotfix_ProjectileAutoKillDistSqExtended)) (Variables|Transient|GetHookshotProjectile))
            (else
              ;;[0.01703 0.1061 0.1094] If there's no valid target under the reticle, check to see if there is a valid target sent by the client. If so, strongly home to that target.
              (bind (_bhome _targetactor_2) (Homing|ShouldHometoClientSuggestedHomingTarget))
              (if _bhome
                (Utilities|IsValid _targetactor_2
                  (:Is%20Valid
                    (bind _relativeoffset (Homing|GetHomingTargetOffset _targetactor_2))
                    (Projectile|SetHomingData (Variables|Transient|GetHookshotProjectile) (Variables|Transient|GetHookshotProjectile) (Utilities|Struct|MakeProjectileHomingData "LockOn" (ScalableFloat|GetValueatLevel (Variables|Hotfix|Homing|Direct|GetHotfix_HomingDirectAim_MinTurnSpeed)) (ScalableFloat|GetValueatLevel (Variables|Hotfix|Homing|Direct|GetHotfix_HomingDirectAim_MaxTurnSpeed)) (ScalableFloat|GetValueatLevel (Variables|Hotfix|Homing|Direct|GetHotfix_HomingDirectAim_TurnSpeedRampTime)) (ScalableFloat|GetValueatLevel (Variables|Hotfix|Homing|Direct|GetHotfix_HomingDirectAim_DistanceThreshold)) -1 _targetactor_2 0 "0.000000,0.000000,0.000000" (Homing|ClampHomingRelativeOffset _relativeoffset) "Relative" true) :!class /Script/FortniteGame.FortProjectileMovementInterface)
                    ;;;[0.224 0.04469 0.2092] Update the distance the projectile travels since we have an accurate aim.
                    ;;; This will help hit targets that are backpedalling away
                    (Variables|/Script/FortniteGame.FortProjectileBase|SetMaxSimulationDistanceSquaredOverride (ScalableFloat|GetValueatLevel (Variables|Hotfix|GetHotfix_ProjectileAutoKillDistSqExtended)) (Variables|Transient|GetHookshotProjectile))
                    (Development|PrintString "Hookshot: Client suggested target used for homing." true false "(R=1.000000,G=0.482353,B=0.000000,A=1.000000)" 1e+01 "Hookshot_EarlyTraceHome"))
                  (:Is%20Not%20Valid
                    _)))
              ;;[0.1042 0.01082 0.09555] If there was never a valid target under the reticle, setup soft homing settings on the projectile based on how close the player's aiming is to the closest potential target.
              (bind (_asynctaskproxy_2 _targetdata_2) (Ability|Tasks|PerformTargetingAbilityTask "Task_HookshotTargeting_Homing" (Variables|Targeting|GetTargetingPresetHookshotHoming) true)
                (:then
                  (Variables|Transient|AsyncTask|SetAsyncTask_HookshotHoming _asynctaskproxy_2))
                (:Targeted
                  _)
                (:Cancelled
                  _))
              (bind (_targetactor_3 _minturnspeed _maxturnspeed _turnspeedramptime _distancethreshold) (Homing|GetPotentialTargetHomingData _targetdata_2))
              (Utilities|IsValid _targetactor_3
                (:Is%20Valid
                  (Development|PrintString "Hookshot: Indirect aim found." true false "(R=1.000000,G=0.482353,B=0.000000,A=1.000000)" 1e+01 "Hookshot")
                  (bind (_relativeoffset_2 _worldoffset) (Homing|GetHomingTargetOffset _targetactor_3))
                  (Projectile|SetHomingData (Variables|Transient|GetHookshotProjectile) (Variables|Transient|GetHookshotProjectile) (Utilities|Struct|MakeProjectileHomingData "LockOn" _minturnspeed _maxturnspeed _turnspeedramptime _distancethreshold -1 _targetactor_3 0 "0.000000,0.000000,0.000000" (Homing|ClampHomingRelativeOffset _relativeoffset_2) "Relative" true) :!class /Script/FortniteGame.FortProjectileMovementInterface))
                (:Is%20Not%20Valid
                  (Development|PrintString "Hookshot: Indirect aim NOT found." true false "(R=1.000000,G=0.482353,B=0.000000,A=1.000000)" 1e+01 "Hookshot"))))))
        (:else
          ;;;[1 0 0] Warning: Projectile null after target selection
          (Fortnite|FortShippingLog self "Projectile null after target selection." true)
          (Ability|EndAbility))))))

(event Custom|ServerSpawnProjectileAtTransform
  (params (SpawnLocation :type Vector) (SpawnDirection :type Rotator))
  ;[0.1421 0.5625 0.5337 0.5] Projectile
  ;;[0.0541 0.2865 0.07439] Spawn the projectile.
  ;;;[0.02404 0.02047 0.15 0.5] ServerSpawnProjectileAtTransform: Received RPC info from the client. Verify the passed info, and use it if successfully verified.
  (Ability|HasAuthority
    (:True
      (bind (_spawnlocation _spawndirection) (Projectile|CalculateProjectileSpawnLocationandRotation)) ; Let the server calculate what it thinks is the best location first.
      (Variables|Transient|SetProjectileSpawnLocation _spawnlocation)
      (Variables|Transient|SetProjectileSpawnDirection _spawndirection)
      (bind (_bsuccess _retspawnlocation _retspawndirection) (General|ValidateClientProjectileSpawnTransformInfo SpawnLocation SpawnDirection)) ; Validate what the client sent in next.\nThis uses the updated server spawn location and direction to verify what the client has sent in.
      (Fortnite|FortShippingLog self (Utilities|String|Append "ValidateClientProjectileSpawnTransformInfo: Client info validation " (select _bsuccess " passed." " failed.")))
      (if _bsuccess
        (Variables|Transient|SetProjectileSpawnLocation _retspawnlocation)
        (Variables|Transient|SetProjectileSpawnDirection _retspawndirection))
      (bind (_asynctaskproxy _projectiledata) (Projectile|Tasks|SpawnProjectileAndWait :Class (Variables|General|GetHookshotProjectileClass) :RequestedBy (CallFunction|GetAvatarasFortPlayerPawn) :SpawnLocation (Variables|Transient|GetProjectileSpawnLocation) :SpawnRotation (Variables|Transient|GetProjectileSpawnDirection) :SpawnDirection (Variables|Transient|GetProjectileSpawnDirection) :InitialSpeed (ScalableFloat|GetValueatLevel (Variables|Hotfix|GetHotfix_ProjectileSpeed)) :GravityScale 0.0 :bAllowProjectileSpawnWhenDead false :bAllowProjectileSpawnWhenDBNO false)
        (:then
          (Variables|Transient|AsyncTask|SetAsyncTask_SpawnProjectile _asynctaskproxy))
        (:Created
          (bind (_projectiledata_spawnedprojectile _projectiledata_hits) (Utilities|Struct|BreakProjectileEventData _projectiledata))
          (bind _asb_prj_hookshot (Utilities|Casting|CastToB_Prj_Hookshot _projectiledata_spawnedprojectile)
            (:then
              (CallFunction|OnProjectileCreated _asb_prj_hookshot))
            (:CastFailed
              ;;;[1 0 0] Warning: Failed to create projectile.
              (Fortnite|FortShippingLog self "Failed to create hookshot projectile." true)
              (Ability|EndAbility))))
        (:Stopped
          (CallFunction|OnProjectileStopped _projectiledata_hits))
        (:Destroyed
          (CallFunction|OnProjectileDestroyed 0))))))

(event Custom|On%20Projectile%20Created
  (params (HookshotProjectile :type B%20Prj%20Hookshot%20Object%20Reference))
  ;[0.1421 0.5625 0.5337 0.5] Projectile
  ;;[0.1089 0.2865 0.1879 0.5] On Projectile Created
  (bind _output_get (Variables|Transient|SetHookshotProjectile HookshotProjectile))
  ;;;[0.1608 0.02747 0.1875] Set up default auto kill distance. Marks how far the projectile will travel.
  (Variables|/Script/FortniteGame.FortProjectileBase|SetMaxSimulationDistanceSquaredOverride (ScalableFloat|GetValueatLevel (Variables|Hotfix|GetHotfix_ProjectileAutoKillDistSqDefault)) _output_get)
  ;;;[1 0.2746 0.9388] On the server, bind to know when the projectile is killed. This fires when the projectile has traveled beyond its kill distance.
  (Default|BindEventtoOnProjectileKilled _output_get (EventDispatchers|CreateEvent :!event Custom|HandeProjectileKilled) :!class /ThornRipper/Gameplay/Items/HookshotGameplay/Gameplay/B_Prj_Hookshot.B_Prj_Hookshot_C)
  ;;;[0.1198 0.04027 0.1302] Cache the rope actor so it can be cleaned up when the ability ends.
  (bind _rope (Class|BPrjHookshot|GetRope _output_get :!class /ThornRipper/Gameplay/Items/HookshotGameplay/Gameplay/B_Prj_Hookshot.B_Prj_Hookshot_C))
  (Variables|Transient|SetRopeActor _rope)
  (Projectile|ApplyPlayerMomentumtoProjectile)
  (Analytics|AnalyticsHookshotThrown)
  (Projectile|HandleHookshotProjectileSpawned (Variables|Transient|GetHookshotProjectile))
  (CallFunction|SetupHomingData))

(event Custom|On%20Projectile%20Stopped
  (params (Projectile%20Hit%20Results :type Array%20of%20Hit%20Result%20Structures))
  ;[0.1421 0.5625 0.5337 0.5] Projectile
  ;;[0 1 0.0106 0.5] Projectile Stopped. Apply Hookshot!
  (Variables|Transient|SetProjectileDataHits Projectile%20Hit%20Results)
  (Ability|HasAuthority
    (:True
      (Default|ApplyHookshot)
      (bind _hookshotprojectile (Variables|Transient|GetHookshotProjectile)
        (:then
          (Default|UnbindEventfromOnProjectileKilled _hookshotprojectile (EventDispatchers|CreateEvent :!event Custom|HandeProjectileKilled) :!class /ThornRipper/Gameplay/Items/HookshotGameplay/Gameplay/B_Prj_Hookshot.B_Prj_Hookshot_C))
        (:else)))))

(event Custom|On%20Projectile%20Destroyed
  (params (Destroyed%20Projectile :type Actor%20Object%20Reference))
  ;[0.1421 0.5625 0.5337 0.5] Projectile
  ;;[0.4323 0.07196 0.03413 0.5] Projectile Destroyed
  (Projectile|DisableProjectile))

(event Custom|ServerSetClientSuggestedHomingTarget
  (params (TargetActor :type Actor%20Object%20Reference))
  ;[1 0.1013 0.7612] ServerSetClientSuggestedHomingTarget: Lets the server know about the target actor that was under the player's reticle when they threw the hook.
  ; The server will use this actor as the homing target if needed.
  ;[There's a chance this RPC call reaches the server a little late to be considered as a homing target under poor network connections but that is OKAY since this is merely a suggestion and the server runs additional traces to figure the best homing target from its perspective]
  (Ability|HasAuthority
    (:True
      (Variables|Transient|SetClientSuggestedHomingTarget TargetActor))))

(event Ability|EventOnEndAbility
  (params (bWasCancelled :type Boolean))
  ;[1 0.344 0] OnEndAbility: Clean up and handle cancelling.
  (Utilities|FlowControl|Sequence
    (:then_0
      (Variables|Transient|SetbAbilityEnded true)
      (Variables|Transient|SetbHookshotThrown false)
      (GameplayTasks|EndTask (Utilities|GetAbilityTasks))
      (Camera|PopCameraMode)
      (Ability|RemoveGameplayEffectFromOwnerWithHandle (Variables|Transient|GetPlayerHookshotReadyGameplayEffectHandle))
      (Variables|Transient|SetbHookshotSuccessful false)
      (Variables|Transient|SetClientSuggestedHomingTarget 0)
      (Variables|Transient|SetbEarlyPrimaryQueued false)
      (Camera|PopCameraMode) ; Remove any lingering camera mode.
      ;;[0.02562 0.125 0.1163] If the hookshot is fired, do not end the hookshot montage and let it play out naturally until finish.
      ;; If the hookshot is NOT fired, stop the hookshot montage with a blend.
      (if (not (Variables|Transient|GetbHookshotFired))
        (bind _outfortplayerpawn (CallFunction|GetAvatarasValidFortPlayerPawn)
          (:ValidFortPlayerPawn
            (bind _mesh (Variables|/Script/Engine.Character|GetMesh _outfortplayerpawn)
              (:then
                (Utilities|IsValid (Components|SkeletalMesh|GetAnimInstance _mesh)
                  (:Is%20Valid
                    (Animation|Montage|MontageStop (Components|SkeletalMesh|GetAnimInstance _mesh) 0.2 (Variables|General|GetHookshotMontage)))
                  (:Is%20Not%20Valid
                    _)))
              (:else
                _)))
          (:AvatarCastFailed
            _)))
      (Variables|Transient|SetbHookshotFired false)
      (bind _ropeactor (Variables|Transient|GetRopeActor)
        (:then
          (Actor|DestroyActor _ropeactor)
          (Variables|Transient|SetRopeActor 0))
        (:else)))
    (:then_1
      ;;[0.2777 0.6 1] Unbind from projectile delegates and clear reference to the hookshot projectile
      (bind _hookshotprojectile (Variables|Transient|GetHookshotProjectile)
        (:then
          (Default|UnbindEventfromOnProjectileKilled _hookshotprojectile (EventDispatchers|CreateEvent :!event Custom|HandeProjectileKilled) :!class /ThornRipper/Gameplay/Items/HookshotGameplay/Gameplay/B_Prj_Hookshot.B_Prj_Hookshot_C)
          (Default|UnbindEventfromOnProjectileStopped _hookshotprojectile (EventDispatchers|CreateEvent :!event Custom|On%20Projectile%20Stopped) :!class /ThornRipper/Gameplay/Items/HookshotGameplay/Gameplay/B_Prj_Hookshot.B_Prj_Hookshot_C)
          (Game|UnbindEventfromOnDestroyed _hookshotprojectile (EventDispatchers|CreateEvent :!event Custom|On%20Projectile%20Destroyed))
          (Actor|DestroyActor _hookshotprojectile)
          (Variables|Transient|SetHookshotProjectile 0))
        (:else)))
    (:then_2
      (|CallParentFunction bWasCancelled))))

(event Custom|HandeProjectileKilled
  ;[0.01535 0.09375 0.01502] Handle the callback on when the projectile is killed for traveling beyond its kill distance.
  ; The server calls an RPC to the client to let it know that the hookshot is complete without a successful pull.
  (Ability|HasAuthority
    (:True
      (Utilities|IsValid self
        (:Is%20Valid
          (CallFunction|ClientHookshotComplete false)
          (Projectile|DisableProjectile)
          (bind _hookshotprojectile (Variables|Transient|GetHookshotProjectile)
            (:then
              (Default|UnbindEventfromOnProjectileKilled _hookshotprojectile (delegate Custom|HandeProjectileKilled) :!class /ThornRipper/Gameplay/Items/HookshotGameplay/Gameplay/B_Prj_Hookshot.B_Prj_Hookshot_C))
            (:else)))))))

(event Custom|ClientHookshotComplete
  (params (bSuccessPull :type Boolean) (bPawnHitDead :type Boolean))
  ;[0.0255 0.04685 0.125] ClientHookshotComplete: Called from the server for the client. Let's the client know that the hookshot projectile behavior has ended with successfully pulling an actor or not. Based on this value, the appropriate section of the animation is played.
  ;;[1 0.5104 0.5104] The thrown projectile hit something, start playing the pull part of the montage.
  (Ability|Animation|MontageJumptoSection (select bSuccessPull (Variables|General|GetPullSuccessMontageSectionName) (Variables|General|GetPullFailMontageSectionName)))
  (Camera|PlayHookshotCameraShake bSuccessPull)
  (if (and (Utilities|IsValidClass (Variables|Camera|GetThrowCameraModeClass)) (not (Variables|Camera|GetbSkipPopThrowCameraOnClientHookshotComplete)))
    (Camera|PopCameraMode) ; Removes the throw camera mode.
  )
  (if (Math|Boolean|NANDBoolean (Variables|Transient|GetbAbilityEnded) (ScalableFloat|GetValueasBool (Variables|Hotfix|GetHotfix_PreventPushCameraAfterAbilityEnd)))
    (Camera|PushCameraMode (Variables|Camera|GetTargetingCameraModeClass)) ; Sets up the targeting camera mode again.
    (else
      (Fortnite|FortShippingLog self "EventGraph: Client Hookshot Complete called after EndAbility " true)))
  (bind _outfortplayerpawn (CallFunction|GetAvatarasValidFortPlayerPawn)
    (:ValidFortPlayerPawn
      (Utilities|IsValid (Fortnite|GetFortPlayerControllerfromActor _outfortplayerpawn)
        (:Is%20Valid
          (Game|Feedback|LocalPlayForceFeedback (Fortnite|GetFortPlayerControllerfromActor _outfortplayerpawn) (select bSuccessPull (Variables|ForceFeedback|GetForceFeedbackHookshotSuccess) (Variables|ForceFeedback|GetForceFeedbackHookshotFail)) false "None"))))))

(fn AnalyticsHookshotThrown)

(fn ApplyHookshot
  (bind _returnvalue (Utilities|Array|LastIndex (Variables|Transient|GetProjectileDataHits)))
  (bind (_bblockinghit _binitialoverlap _time _distance _location _impactpoint _normal _impactnormal _physmat _hitactor) (Collision|BreakHitResult (LocalVariables|GetProjectileHitResult)))
  (bind _projectilehitresult (LocalVariables|GetProjectileHitResult))
  (bind (_bblockinghit_2 _binitialoverlap_2 _time_2 _distance_2 _location_2 _impactpoint_2 _normal_2 _impactnormal_2 _physmat_2 _hitactor_2) (Collision|BreakHitResult _projectilehitresult))
  (bind (_bblockinghit_3 _binitialoverlap_3 _time_3 _distance_3 _location_3 _impactpoint_3 _normal_3 _impactnormal_3 _physmat_3 _hitactor_3) (Collision|BreakHitResult (LocalVariables|GetProjectileHitResult)))
  (bind _projectilehitresult_2 (LocalVariables|GetProjectileHitResult))
  (bind _returnvalue_2 (Ability|TargetData|AbilityTargetDatafromHitResult _projectilehitresult_2))
  (bind _projectilehitresult_3 (LocalVariables|GetProjectileHitResult))
  (bind (_bblockinghit_4 _binitialoverlap_4 _time_4 _distance_4 _location_4 _impactpoint_4 _normal_4 _impactnormal_4 _physmat_4 _hitactor_4 _hitcomponent _hitbonename _bonename _hititem _elementindex _faceindex _tracestart _traceend) (Collision|BreakHitResult _projectilehitresult_3))
  (bind (_asfort_pawn _bsuccess) (Utilities|Casting|CastToFortPawn _hitactor_3))
  (bind _projectiledatahits (Variables|Transient|GetProjectileDataHits))
  (bind _projectilehitresult_4 (LocalVariables|GetProjectileHitResult))
  ;[0.02595 0.1284 0.1458] ApplyHookshot: Invoked when the projectile touches (hits/overlaps) an actor. The information of the touch is stored within the ProjectileDataHits array.
  ; Only care about the blocking last touched actor (last index).
  ; Attempt to apply the gameplay effect container which will eventually apply the pull GA if the actor is valid.
  ; If the pull GA was applied successfully, consider the target to be pulled successfully and fire RPC.
  (Fortnite|FortShippingLog self "ApplyHookshot: Invoked")
  (if (Utilities|Array|IsValidIndex _projectiledatahits _returnvalue)
    (bind _output_get (LocalVariables|SetProjectileHitResult (Utilities|Array|Get(aref) _projectiledatahits _returnvalue)))
    (if (Collision|BreakHitResult _output_get)
      (Utilities|FlowControl|Sequence
        (:then_0
          ;[0.1389 0.2057 1] If we hit an item owned by a pawn, then change the hit actor to the attached parent pawn.
          ; This fixes a bug when a pawn's SunRose Wings were hit damage numbers would be applied twice resulting in double the value being displayed.
          ;
          ; This is because calling ApplyGameplayEffectContainer() with an attached object will cause the damage numbers to be added again (damage numbers should only be added by ExecuteGameplayCueOnOwner().) The exact reasons for this are deep in the vanilla ability code and everything seems to make sense within its own isolated context so calling ApplyGameplayEffectContainer() with a pawn instead of the hit weapon seems like a good solution.
          (Utilities|IsValid _hitactor_4
            (:Is%20Valid
              (Utilities|Casting|CastToPawn _hitactor_4
                (:CastFailed
                  (bind _aspawn (Utilities|Casting|CastToPawn (Actor|GetAttachParentActor _hitactor_4)))
                  (LocalVariables|SetProjectileHitResult (Collision|MakeHitResult _bblockinghit_4 _binitialoverlap_4 _time_4 _distance_4 _location_4 _impactpoint_4 _normal_4 _impactnormal_4 _physmat_4 _aspawn _hitcomponent _hitbonename _bonename _hititem _elementindex _faceindex _tracestart _traceend)))))))
        (:then_1
          ;;[0.1365 0.02464 0.1406] Responsible for handling the hit impact fxs through gameplay cues.
          (bind _asfort_ability_system_component (Utilities|Casting|CastToFortAbilitySystemComponent (Ability|GetAbilitySystemComponentfromActorInfo)))
          (bind _output_get_2 (LocalVariables|SetEffectContextHandle (GameplayEffects|MakeFortEffectContext _asfort_ability_system_component 0 (CallFunction|GetAvatarasFortPlayerPawn) (Variables|Transient|GetHookshotProjectile) 1)))
          (Ability|EffectContext|AddHitResult _output_get_2 _projectilehitresult_4 true)
          (Ability|ExecuteGameplayCueOnOwner (select _bsuccess (Variables|GameplayCues|GetGameplayCueTagImpactPawn) (Variables|GameplayCues|GetGameplayCueTagImpactWorld)) _output_get_2))
        (:then_2
          (Projectile|DisableProjectile _projectilehitresult_2)
          (Ability|ApplyGameplayEffectContainer _returnvalue_2 (Variables|General|GetHookshotPullApplicationTag))
          (Ability|ApplyGameplayEffectContainer _returnvalue_2 (Variables|General|GetHookshotDamageApplicationTag) (Level|GetHookshotDamageGameplayEffectLevel))
          (Ability|SendGameplayEventtoActor (CallFunction|GetAvatarasFortPlayerPawn) (Variables|Default|GetHookedActorEventTag) (Utilities|Struct|MakeGameplayEventData (Variables|Default|GetHookedActorEventTag) 0 _hitactor))
          (Fortnite|FortShippingLog self (Utilities|String|Append "ApplyHookshot: Applying hookshot pull and damage gameplay effect to: " (Utilities|GetDisplayName _hitactor_2)))
          (Utilities|IsValid self
            (:Is%20Valid
              (bind _asfort_pawn_2 (Utilities|Casting|CastToFortPawn _hitactor_2)
                (:then
                  (if (FortPawn|IsDead _asfort_pawn_2)
                    (Variables|Transient|SetbHookshotSuccessful true)
                    (LocalVariables|SetPawnHitDead true)
                    (else
                      (bind _returnvalue_3 (GameplayTags|HasMatchingGameplayTag(Message) _hitactor_2 "(TagName=\"Gameplay.Status.Pulled\")" :!label _label))
                      (Variables|Transient|SetbHookshotSuccessful _returnvalue_3)))
                  (CallFunction|ClientHookshotComplete (Variables|Transient|GetbHookshotSuccessful) (LocalVariables|GetPawnHitDead)))
                (:CastFailed
                  _))
              (goto _label))))))
    (else
      (Utilities|IsValid self
        (:Is%20Valid
          ;;[1 0.2311 0.6877] ClientHookshotComplete needs to be called even if the ProjectileDataHits is empty so that the ability continues.
          (CallFunction|ClientHookshotComplete false))))))

(fn ApplyPlayerMomentumToProjectile
  ;[0.06758 0.224 0.2195] ApplyPlayerMomentumToProjectile: Takes a part of the player's perpendicular velocity at the time of throw and applies it to the projectile.
  (bind _outfortplayerpawn (CallFunction|GetAvatarasValidFortPlayerPawn)
    (:ValidFortPlayerPawn
      (bind _hookshotprojectile (Variables|Transient|GetHookshotProjectile)
        (:then
          (bind _projectilemovementcomponent (Variables|/Script/FortniteGame.FortProjectileBase|GetProjectileMovementComponent _hookshotprojectile)
            (:then
              (bind (_returnvalue_x _returnvalue_y) (Math|Vector|BreakVector (Utilities|Operators|Multiply (Math|Vector|DotProduct (General|GetCharacterVelocity _outfortplayerpawn) (Math|Vector|Normalize (Math|Vector|RotateVectorAroundAxis (Variables|/Script/Engine.MovementComponent|GetVelocity _projectilemovementcomponent) 9e+01 (Math|Vector|GetUpVector (Math|Rotator|MakeRotfromX (Variables|/Script/Engine.MovementComponent|GetVelocity _projectilemovementcomponent)))))) (Math|Vector|Normalize (Math|Vector|RotateVectorAroundAxis (Variables|/Script/Engine.MovementComponent|GetVelocity _projectilemovementcomponent) 9e+01 (Math|Vector|GetUpVector (Math|Rotator|MakeRotfromX (Variables|/Script/Engine.MovementComponent|GetVelocity _projectilemovementcomponent))))) (ScalableFloat|GetValueatLevel (Variables|Hotfix|GetHotfix_PlayerMomentumToPrjMultiplier)))))
              (Projectile|SetVelocity _projectilemovementcomponent (Utilities|Operators|Add (Variables|/Script/Engine.MovementComponent|GetVelocity _projectilemovementcomponent) (Math|Vector|MakeVector _returnvalue_x _returnvalue_y))))
            (:else)))
        (:else)))))

(fn CalculateProjectileSpawnLocationAndRotation
  ;[0.03455 0.119 0.15 0.5] Calculates and returns the location at which the hookshot projectile should spawn.
  (Utilities|FlowControl|Sequence
    (:then_0
      (bind _outfortplayerpawn (CallFunction|GetAvatarasValidFortPlayerPawn)
        (:ValidFortPlayerPawn
          (bind _returnvalue (TargetingSystem|GetPlayerCameraTargetingTransform _outfortplayerpawn))
          (bind _adjustedlocation (Projectile|GetCharacterVelocityAdjustedProjectileSpawnLocation (.location _returnvalue)))
          (bind _adjustedlocation_2 (Projectile|GetTraceAdjustedSpawnLocation _adjustedlocation))
          (return _adjustedlocation_2 (.rotation _returnvalue)))))
    (:then_1
      (return "0, 0, 0" "0, 0, 0"))))

(fn ClampHomingRelativeOffset
  (params (InVec :type Vector))
  ;[0.005351 0.07292 0.06345] Clamp the relative Z offset to reduce the chance that the projectile hits landscape under the player.
  (return (Math|Vector|MakeVector (.x InVec) (.y InVec) (Math|Float|Max(Float) (.z InVec) (ScalableFloat|GetValueatLevel (Variables|Hotfix|GetHotfix_ClampHomingRelativeOffsetMaxZ))))))

(fn DisableProjectile
  (params (HitResult :type Hit%20Result%20Structure))
  ;[0.09106 0.02935 0.1042] DisableProjectile: Hides the projectile, disables the collision, and let's the rope know that it is time to end its behavior.
  (Ability|HasAuthority
    (:True
      (bind _hookshotprojectile (Variables|Transient|GetHookshotProjectile)
        (:then
          (Class|BPrjHookshot|DisableProjectile _hookshotprojectile HitResult :!class /ThornRipper/Gameplay/Items/HookshotGameplay/Gameplay/B_Prj_Hookshot.B_Prj_Hookshot_C))
        (:else)))))

(fn GetCharacterVelocity
  (params (Character :type Character%20Object%20Reference))
  ;[0.02249 0.1354 0.1314] GetCharacterVelocity: Returns the velocity of the character after accounting for the base movement component.
  (Utilities|FlowControl|Sequence
    (:then_0
      (Utilities|IsValid Character
        (:Is%20Valid
          (bind _charactermovement (Variables|/Script/Engine.Character|GetCharacterMovement Character)
            (:then
              (bind _asprimitive_component (Utilities|Casting|CastToPrimitiveComponent (Pawn|Components|CharacterMovement|GetMovementBaseObject _charactermovement))
                (:then
                  (LocalVariables|SetMovementBaseVelocity (Transformation|GetComponentVelocity _asprimitive_component)))
                (:CastFailed
                  _)))
            (:else
              _))
          (return (+ (LocalVariables|GetMovementBaseVelocity) (Transformation|GetVelocity Character))))))
    (:then_1
      (return "0, 0, 0"))))

(fn GetCharacterVelocityAdjustedProjectileSpawnLocation
  (params (SpawnLocation :type Vector))
  (bind _returnvalue (CallFunction|GetAvatarasFortPlayerPawn))
  ;[0.08854 0.01285 0.07211] GetCharacterVelocityAdjustedProjectileSpawnLocation: Adjust the spawn location of the projectile based on how fast the player is moving. Calculation is done with ping in consideration to better match the client side expectation. This logic helps fire the projectile towards the player's crosshair when they are moving at high speeds - e.g., standing on the moving train.
  (Utilities|FlowControl|Sequence
    (:then_0
      (Ability|HasAuthority
        (:True
          (if (ScalableFloat|GetValueasBool (Variables|Hotfix|GetHotfix_AllowAdjustSpawnLocationBasedOnVelocity))
            (Utilities|IsValid _returnvalue
              (:Is%20Valid
                (bind _asfort_player_state_athena (Utilities|Casting|CastToFortPlayerStateAthena (Variables|/Script/Engine.Pawn|GetPlayerState _returnvalue)))
                (return (Math|Vector|MakeVector (+ (.x SpawnLocation) (.x (* (General|GetCharacterVelocity _returnvalue) (* (* (Math|Float|Min(Float) (PlayerState|GetPinginMilliseconds _asfort_player_state_athena) 2.5e+02) (ScalableFloat|GetValueatLevel (Variables|Hotfix|GetHotfix_AllowAdjustSpawnLocationBasedOnVelocityPingMultiplier))) 0.001)))) (+ (.y SpawnLocation) (.y (* (General|GetCharacterVelocity _returnvalue) (* (* (Math|Float|Min(Float) (PlayerState|GetPinginMilliseconds _asfort_player_state_athena) 2.5e+02) (ScalableFloat|GetValueatLevel (Variables|Hotfix|GetHotfix_AllowAdjustSpawnLocationBasedOnVelocityPingMultiplier))) 0.001)))) (.z SpawnLocation)))))))))
    (:then_1
      (return SpawnLocation))))

(fn GetClientSuggestedHomingTargetFromTargetingComponent
  ;[0.02209 0.15 0.08269 0.5] Should be overridden in the children based on the specific targeting component setup.
  (return 0))

(fn GetHomingTargetFromTargetData
  (params (TargetData :type Gameplay%20Ability%20Target%20Data%20Handle%20Structure%20%28by%20ref%29))
  (bind _returnvalue (Ability|TargetData|GetHitResultfromTargetData TargetData))
  (bind (_bblockinghit _binitialoverlap _time _distance _location _impactpoint _normal _impactnormal _physmat _hitactor) (Collision|BreakHitResult _returnvalue))
  (bind _returnvalue_2 (Actor|GetAttachParentActor _hitactor))
  ;[0.06374 0.2144 0.2188] GetHomingTargetFromTargetData: Finds the appropriate target actor from the target data.
  (Utilities|FlowControl|Sequence
    (:then_0
      (Utilities|IsValid _hitactor
        (:Is%20Valid
          (bind _bvalid (Homing|IsValidHomingTarget _hitactor))
          (if _bvalid
            (return _hitactor)
            (else
              ;;[0.1042 0.02247 0.1037] Check if the hit actor is attached to a valid homing target - e.g., wings on the player pawn.
              (bind _bvalid_2 (Homing|IsValidHomingTarget _returnvalue_2))
              (if _bvalid_2
                (return _returnvalue_2)))))))
    (:then_1
      (return 0))))

(fn GetHomingTargetOffset
  (params (TargetActor :type Actor%20Object%20Reference))
  ;[0.03514 0.1615 0.155] GetHomingTargetOffset: Return the target offset to be used for the homing - closest location on the target collision to the crosshair / aim direction!
  (Utilities|FlowControl|Sequence
    (:then_0
      (bind _outfortplayerpawn (CallFunction|GetAvatarasValidFortPlayerPawn)
        (:ValidFortPlayerPawn
          (bind _ascharacter (Utilities|Casting|CastToCharacter TargetActor))
          (bind _capsulecomponent (Variables|/Script/Engine.Character|GetCapsuleComponent _ascharacter)
            (:then
              (bind _returnvalue (Utility|GetAimLocationNearestPawn _outfortplayerpawn))
              (bind (_returnvalue_location _returnvalue_rotation) (Math|Transform|BreakTransform (Ability|GetTargetingTransform _outfortplayerpawn "Camera" self)))
              (bind _output_get (LocalVariables|SetWorldOffsetLocation (Collision|GetClosestPointonCollision _capsulecomponent (Math|Vector|FindClosestPointonLine (Transformation|GetActorLocation _ascharacter) _returnvalue (Math|Vector|GetForwardVector _returnvalue_rotation)) "None")))
              (return (Math|Transform|InverseTransformLocation (Transformation|GetActorTransform _ascharacter) _output_get) _output_get))
            (:else)))))
    (:then_1
      (return "0, 0, 0" "0, 0, 0"))))

(fn GetHookshotDamageGameplayEffectLevel
  ;[0.001532 0.1443 0.15 0.5] Can be overridden in children.
  (return 1))

(fn GetPotentialTargetHomingData
  (params (TargetSelection :type Gameplay%20Ability%20Target%20Data%20Handle%20Structure))
  (bind _returnvalue (CallFunction|GetAvatarasFortPlayerPawn))
  (bind (_returnvalue_location _returnvalue_rotation) (Math|Transform|BreakTransform (Ability|GetTargetingTransform _returnvalue "Camera" self)))
  ;[0.01664 0.125 0.1175] GetPotentialTargetHomingData: There's no target under the player's reticle. This possibly means that they are trying to lead the aim on a moving target. This function attempts to verify if there's a target that the projectile should home towards.
  (Utilities|FlowControl|Sequence
    (:then_0
      ;;[0.02592 0.1094 0.06463] Only consider the closest player.
      (bind _targetactor (Homing|GetHomingTargetfromTargetData TargetSelection))
      (bind _ascharacter (Utilities|Casting|CastToCharacter _targetactor))
      (bind _output_get (LocalVariables|SetTargetCharacter _ascharacter))
      (bind _bvalid (Homing|IsValidHomingTarget _output_get))
      (if _bvalid
        (bind _capsulecomponent (Variables|/Script/Engine.Character|GetCapsuleComponent _output_get)
          (:then
            (LocalVariables|SetTargetCharacterCapsuleComponent _capsulecomponent)
            ;;[0.1927 0.03463 0.1654] Check line of sight to the potential target.
            (bind _outfortplayerpawn (CallFunction|GetAvatarasValidFortPlayerPawn)
              (:ValidFortPlayerPawn
                (bind _returnvalue_2 (Utility|GetAimLocationNearestPawn _outfortplayerpawn))
                (bind _returnvalue_3 (Fortnite|CheckLineOfSightToActorWithChannel _returnvalue_2 (LocalVariables|GetTargetCharacter) "ECC_Visibility" _outfortplayerpawn))
                (if _returnvalue_3
                  ;;[0.4792 0.4225 0.1174] Find the location where the target is going to be by the time the projectile gets to them.
                  (bind _output_get_2 (LocalVariables|SetPredictedTargetLocation (+ (Transformation|GetActorLocation (LocalVariables|GetTargetCharacter)) (* (Math|Float|MapRangeClamped (Math|Vector|Distance(Vector) _returnvalue_2 (Transformation|GetActorLocation (LocalVariables|GetTargetCharacter))) 0.0 (Variables|Transient|GetCachedHookshotDistance) 0.0 (ScalableFloat|GetValueatLevel (Variables|Hotfix|GetHotfix_Projectile_PreCalculatedLifetime))) (General|GetCharacterVelocity (LocalVariables|GetTargetCharacter))))))
                  ;;[0.1563 0.03432 0.1615] Calculate how good the player's lead targeting is based on an angle between the crosshair and the predicted target location..
                  (bind _returnvalue_4 (Utility|GetAimLocationNearestPawn _returnvalue))
                  (bind _returnvalue_5 (Fortnite|GetAngleDegrees (Math|Vector|GetForwardVector _returnvalue_rotation) (Math|Vector|Normalize (- _output_get_2 _returnvalue_4))))
                  (Development|PrintString (Utilities|String|Append "Hookshot: Indirect aim angle:  " (Utilities|String|ToString(Float) _returnvalue_5)) true false "(R=1.000000,G=0.482353,B=0.000000,A=1.000000)" 1e+01 "Hookshot_IndirectAngle")
                  ;;[0.007072 0.1562 0] If the angle value is less than the max limit, then set calculate homing settings based on that angle amount.
                  (bind _output_get_3 (LocalVariables|SetMaxAllowedAngle (ScalableFloat|GetValueatLevel (Variables|Hotfix|Homing|Indirect|GetHotfix_HomingIndirectAim_MaxAllowedAngle))))
                  (if (> _output_get_3 _returnvalue_5)
                    (Development|PrintString "Hookshot: Indirect aim angle valid." true false "(R=1.000000,G=0.482353,B=0.000000,A=1.000000)" 1e+01 "Hookshot_IndirectAngle_Result")
                    (return (LocalVariables|GetTargetCharacter) (Math|Float|MapRangeClamped _returnvalue_5 0.0 (LocalVariables|GetMaxAllowedAngle) (ScalableFloat|GetValueatLevel (Variables|Hotfix|Homing|Indirect|GetHotfix_HomingIndirectAim_MinTurnSpeed_Min)) (ScalableFloat|GetValueatLevel (Variables|Hotfix|Homing|Indirect|GetHotfix_HomingIndirectAim_MinTurnSpeed_Max))) (Math|Float|MapRangeClamped _returnvalue_5 0.0 (LocalVariables|GetMaxAllowedAngle) (ScalableFloat|GetValueatLevel (Variables|Hotfix|Homing|Indirect|GetHotfix_HomingIndirectAim_MaxTurnSpeed_Min)) (ScalableFloat|GetValueatLevel (Variables|Hotfix|Homing|Indirect|GetHotfix_HomingIndirectAim_MaxTurnSpeed_Max))) (Math|Float|MapRangeClamped _returnvalue_5 0.0 (LocalVariables|GetMaxAllowedAngle) (ScalableFloat|GetValueatLevel (Variables|Hotfix|Homing|Indirect|GetHotfix_HomingIndirectAim_MinTurnSpeedRampTime_Max)) (ScalableFloat|GetValueatLevel (Variables|Hotfix|Homing|Indirect|GetHotfix_HomingIndirectAim_MinTurnSpeedRampTime_Min))) (Math|Float|MapRangeClamped _returnvalue_5 0.0 (LocalVariables|GetMaxAllowedAngle) (ScalableFloat|GetValueatLevel (Variables|Hotfix|Homing|Indirect|GetHotfix_HomingIndirectAim_DistanceThreshold_Min)) (ScalableFloat|GetValueatLevel (Variables|Hotfix|Homing|Indirect|GetHotfix_HomingIndirectAim_DistanceThreshold_Max)))))))))
          (:else))))
    (:then_1
      (return 0 0.0 0.0 0.0 0.0))))

(fn GetProjectileDataHits
  (return (Variables|Transient|GetProjectileDataHits)))

(fn GetTraceAdjustedSpawnLocation
  (params (InSpawnLocation :type Vector))
  (bind (_rope _onprojectilekilled _hotfix_projectileradius) (Class|GetClassDefaults (Variables|General|GetHookshotProjectileClass)))
  ;[0.02778 0.1667 0.1667] GetTraceAdjustedSpawnLocation: Adjust the spawn location of the projectile based on a trace so that the projectile does not spawn on the other side of a wall when the camera is too close to it.
  (Utilities|FlowControl|Sequence
    (:then_0
      (if (ScalableFloat|GetValueasBool (Variables|Hotfix|GetHotfix_AllowAdjustSpawnLocationBasedOnTrace))
        (bind _outfortplayerpawn (CallFunction|GetAvatarasValidFortPlayerPawn)
          (:ValidFortPlayerPawn
            (bind _output_get (LocalVariables|SetActorsToIgnore (Actor|GetAttachedActors _outfortplayerpawn true true)))
            (Utilities|Array|Add _output_get _outfortplayerpawn)
            (Utilities|Array|Add _output_get (Vehicle|GetVehicleActor _outfortplayerpawn :!class /Script/FortniteGame.FortPlayerPawn))
            (bind (_outhit _returnvalue) (Collision|LineTraceByChannel (Transformation|GetActorLocation _outfortplayerpawn) InSpawnLocation "TraceTypeQuery5" true (LocalVariables|GetActorsToIgnore) "ForDuration"))
            (if _returnvalue
              (if (ScalableFloat|GetValueasBool (Variables|Hotfix|GetHotfix_AdustSpawnLocationByProjSize))
                (return (* (Math|Vector|Normalize (- (.TraceStart (Collision|BreakHitResult _outhit)) (.ImpactPoint (Collision|BreakHitResult _outhit)))) (Math|Float|Clamp(Float) (+ (ScalableFloat|GetValueatLevel _hotfix_projectileradius) (ScalableFloat|GetValueatLevel (Variables|Hotfix|GetHotfix_AdjustSpawnProjOffset))) 0.0 (.Distance (Collision|BreakHitResult _outhit)))))
                (else
                  (return (.ImpactPoint (Collision|BreakHitResult _outhit))))))))))
    (:then_1
      (return InSpawnLocation))))

(fn HandleHookshotProjectileSpawned
  (params (HookshotPrj :type B%20Prj%20Hookshot%20Object%20Reference)))

(fn IsBot
  (bind _outfortplayerpawn (CallFunction|GetAvatarasValidFortPlayerPawn)
    (:ValidFortPlayerPawn
      (return (Pawn|IsBotControlled _outfortplayerpawn)))
    (:AvatarCastFailed
      (return false))))

(fn IsValidHomingTarget
  (params (TargetActor :type Actor%20Object%20Reference))
  (bind _returnvalue (CallFunction|GetAvatarasFortPlayerPawn))
  (bind (_asfort_player_pawn _bsuccess) (Utilities|Casting|CastToFortPlayerPawn TargetActor))
  (bind (_asfort_aipawn _bsuccess_2) (Utilities|Casting|CastToFortAIPawn TargetActor))
  (bind _hookshotprojectile (Variables|Transient|GetHookshotProjectile))
  ;[0.01073 0.1702 0.1719] IsValidHomingTarget: Only home towards player pawns and AI pawns that are hostile or neutral and in front of projectile
  (Utilities|FlowControl|Sequence
    (:then_0
      (if (or _bsuccess _bsuccess_2)
        (Utilities|IsValid _returnvalue
          (:Is%20Valid
            (if (ScalableFloat|GetValueasBool (Variables|Hotfix|GetHotfix_IgnoreFriendlyPawns))
              (if (Utilities|Enum|NotEqual(Enum) (Team|GetTeamAffiliation _returnvalue TargetActor) "Friendly")
                ;;[0.349 0.2389 0.0623 0.5] Target is invalid if projectile is has already past it
                (if (and (ScalableFloat|GetValueasBool (Variables|Hotfix|Homing|GetHotfix_IgnoreHomingTargetsBehindProjectile)) (Utilities|IsValid _hookshotprojectile :!class /Script/Engine.KismetSystemLibrary))
                  (return (>= (Math|Vector|DotProduct (- (Transformation|GetActorLocation (Variables|GetTargetActor)) (Transformation|GetActorLocation _hookshotprojectile)) (Transformation|GetVelocity _hookshotprojectile)) 0))
                  (else
                    (return true))))
              (else
                (return true)))))))
    (:then_1
      (return false))))

(fn PlayHookshotCameraShake
  (params (bSuccessfulHookshot :type Boolean))
  (bind _hookshotsuccessfulcamerashake (Variables|Camera|GetHookshotSuccessfulCameraShake))
  (bind _hookshotmisscamerashake (Variables|Camera|GetHookshotMissCameraShake))
  (bind _returnvalue (select bSuccessfulHookshot _hookshotsuccessfulcamerashake _hookshotmisscamerashake))
  (Ability|HasAuthority
    (:False
      (bind _outfortplayerpawn (CallFunction|GetAvatarasValidFortPlayerPawn)
        (:ValidFortPlayerPawn
          (bind _returnvalue_2 (Fortnite|GetFortPlayerControllerfromActor _outfortplayerpawn))
          (Utilities|IsValid _returnvalue_2
            (:Is%20Valid
              (if (Utilities|IsValidClass _returnvalue)
                (Game|Feedback|ClientStartCameraShake :self _returnvalue_2 :Shake _returnvalue :Scale 1.0 :PlaySpace "CameraLocal")))))))))

(fn PlayThrowCameraShake
  (bind _throwcamerashake (Variables|Camera|GetThrowCameraShake))
  (Ability|HasAuthority
    (:False
      (bind _outfortplayerpawn (CallFunction|GetAvatarasValidFortPlayerPawn)
        (:ValidFortPlayerPawn
          (bind _returnvalue (Fortnite|GetFortPlayerControllerfromActor _outfortplayerpawn))
          (Utilities|IsValid _returnvalue
            (:Is%20Valid
              (if (Utilities|IsValidClass _throwcamerashake)
                (Game|Feedback|ClientStartCameraShake :self _returnvalue :Shake _throwcamerashake :Scale 1.0 :PlaySpace "CameraLocal")))))))))

(fn ShouldHomeToClientSuggestedHomingTarget
  (bind _clientsuggestedhomingtarget (Variables|Transient|GetClientSuggestedHomingTarget))
  (bind _returnvalue (Transformation|GetActorLocation _clientsuggestedhomingtarget))
  (bind _self self)
  (bind _aimlocationnearestpawn (LocalVariables|GetAimLocationNearestPawn))
  (bind _clientsuggestedhomingtarget_2 (Variables|Transient|GetClientSuggestedHomingTarget))
  ;[0.009441 0.0625 0.05689] ShouldHomeToClientSuggestedHomingTarget: Returns if the hookshot projectile should home to the target suggested by the client.
  (Utilities|FlowControl|Sequence
    (:then_0
      ;;[0.1562 0.05675 0.08191] First check to see if the suggested homing target is valid to home towards.
      (bind _bvalid (Homing|IsValidHomingTarget _clientsuggestedhomingtarget_2))
      (if _bvalid
        ;;[0.3229 0.09715 0.223] Check to see if the target is at a mimimum distance from the avatar player
        (bind _outfortplayerpawn (CallFunction|GetAvatarasValidFortPlayerPawn)
          (:ValidFortPlayerPawn
            (bind _returnvalue_2 (Utility|GetAimLocationNearestPawn _outfortplayerpawn))
            (bind _output_get (LocalVariables|SetAimLocationNearestPawn _returnvalue_2))
            (if (> (Math|Vector|DistanceSquared(Vector) _output_get (Transformation|GetActorLocation _clientsuggestedhomingtarget_2)) (Math|Float|Square (ScalableFloat|GetValueatLevel (Variables|Hotfix|GetHotfix_MinDistanceToPlayerForHoming))))
              ;;[0.02715 0.1667 0.128] Check to see if the target is under the max distance from camera forward vector.
              (bind (_returnvalue_location _returnvalue_rotation) (Math|Transform|BreakTransform (Ability|GetTargetingTransform _outfortplayerpawn "Camera" _self)))
              (Development|PrintString (Utilities|String|Append "Hookshot: Target distance from aim: " (Utilities|String|ToString(Float) (Math|Vector|GetPointDistancetoLine _returnvalue _aimlocationnearestpawn (Math|Vector|GetForwardVector _returnvalue_rotation)))) true false "(R=1.000000,G=0.482353,B=0.000000,A=1.000000)" 1e+01 "Hookshot_CrosshairDistCheck")
              (if (<= (Math|Vector|GetPointDistancetoLine _returnvalue _aimlocationnearestpawn (Math|Vector|GetForwardVector _returnvalue_rotation)) (ScalableFloat|GetValueatLevel (Variables|Hotfix|GetHotfix_MaxDistanceFromCameraDistForHoming)))
                ;;[0.1347 0.5208 0.4104] Check to see if avatar player still has line of sight to the target actor.
                (bind _returnvalue_3 (Fortnite|CheckLineOfSightToActorWithChannel (LocalVariables|GetAimLocationNearestPawn) (Variables|Transient|GetClientSuggestedHomingTarget) "ECC_Visibility" (CallFunction|GetAvatarasFortPlayerPawn)))
                (if _returnvalue_3
                  (return true (Variables|Transient|GetClientSuggestedHomingTarget))
                  (else
                    (Development|PrintString "Hookshot: Home to client suggest: FAIL: No LoS" true false "(R=1.000000,G=0.482353,B=0.000000,A=1.000000)" 1e+01 "Hookshot_ShouldHomeToClientSuggest")))))))))
    (:then_1
      (return false 0))))

(fn ValidateClientProjectileSpawnTransformInfo
  (params (SpawnLocation :type Vector) (SpawnDirection :type Rotator))
  ;[0.02843 0.15 0.1181 0.5] ValidateClientProjectileSpawnTransformInfo: Validate on the server if the client value passed through the RPC is valid.
  ; Currently, the logic checks to verify that the info passed is not too far off from what the server has cached.
  (Utilities|FlowControl|Sequence
    (:then_0
      (if (ScalableFloat|GetValueasBool (Variables|Hotfix|GetHotfix_ClientRPCProjectileSpawnTransform))
        (Variables|Transient|GetHookshotProjectile
          (:then
            (Fortnite|FortShippingLog self "ValidateClientProjectileSpawnTransformInfo: Attempting to spawn projectile when one already exists. Failed Validation" true))
          (:else
            (LocalVariables|SetVerifyMaxDistanceBetweenServerClientSpawnLocations (ScalableFloat|GetValueatLevel (Variables|Hotfix|GetHotfix_ClientRPCProjectileSpawnLocationVerifyMaxDistance)))
            (LocalVariables|SetVerifyMaxAngleBetweenServerClientDirections (ScalableFloat|GetValueatLevel (Variables|Hotfix|GetHotfix_ClientRPCProjectileSpawnDirectionVerifyMaxAngle)))
            (LocalVariables|SetDistanceBetweenServerClientSpawnLocations (Math|Vector|Distance(Vector) (Variables|Transient|GetProjectileSpawnLocation) SpawnLocation))
            (LocalVariables|SetAngleBetweenServerClientSpawnDirections (Math|Float|Absolute(Float) (Math|Trig|Acos(Degrees) (Math|Vector|DotProduct (Math|Vector|GetForwardVector SpawnDirection) (Math|Vector|GetForwardVector (Variables|Transient|GetProjectileSpawnDirection))))))
            (Fortnite|FortShippingLog self (Utilities|String|Append "ValidateClientProjectileSpawnTransformInfo: " "VerifyMaxDistanceBetweenServerClientSpawnLocations [" (Utilities|String|ToString(Float) (LocalVariables|GetVerifyMaxDistanceBetweenServerClientSpawnLocations)) "] ; VerifyMaxAngleBetweenServerClientDirections [" (Utilities|String|ToString(Float) (LocalVariables|GetVerifyMaxAngleBetweenServerClientDirections)) "] ; DistanceBetweenServerClientSpawnLocations [" (Utilities|String|ToString(Float) (LocalVariables|GetDistanceBetweenServerClientSpawnLocations)) "] ; AngleBetweenServerClientSpawnDirections [" (Utilities|String|ToString(Float) (Math|Float|Absolute(Float) (LocalVariables|GetAngleBetweenServerClientSpawnDirections))) "]"))
            (if (and (<= (LocalVariables|GetDistanceBetweenServerClientSpawnLocations) (LocalVariables|GetVerifyMaxDistanceBetweenServerClientSpawnLocations)) (<= (Math|Float|Absolute(Float) (LocalVariables|GetAngleBetweenServerClientSpawnDirections)) (LocalVariables|GetVerifyMaxAngleBetweenServerClientDirections)))
              ;; Running a trace to ensure that we're not spawning a projectile on another side of a 'wall' or 'object'.
              (bind _outhit (Collision|LineTraceByProfile (Variables|Transient|GetProjectileSpawnLocation) (Variables|GetSpawnLocation) "FortProjectileHitAllPawns" true (Utilities|Array|MakeArray (CallFunction|GetAvatarasFortPlayerPawn)) "ForDuration"))
              (Fortnite|FortShippingLog self (Utilities|String|Append "ValidateClientProjectileSpawnTransformInfo: Trace check from Server calculated spawn location to client sent spawn location " (select (Collision|BreakHitResult _outhit) (Utilities|String|Append " hit an object " (Utilities|GetDisplayName (.HitActor (Collision|BreakHitResult _outhit)))) "hit nothing.")))
              (return true (select (Collision|BreakHitResult _outhit) (.ImpactPoint (Collision|BreakHitResult _outhit)) (Variables|GetSpawnLocation)) (Variables|GetSpawnDirection)))))
        (else
          (Fortnite|FortShippingLog self "ValidateClientProjectileSpawnTransformInfo: Received validation check even though ClientRPCProjectileSpawnTransform feature is disabled." true))))
    (:then_1
      (return false "0, 0, 0" "0, 0, 0"))))
